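# Sound cue for shaking a vinyl present in the Winterfest event, assigned to the Fort_UI sound class with boosted volume.
# A short fixed delay leads into a Modulator for pitch and volume variation, fed by a Random node that picks among four equally weighted shake waves.
# The Random node is excluded from branch culling, so all four variations stay available.
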SoundCue /Game/Athena/Sounds/Events/Winterfest/SC_Winterfest_Vinyl_Present_Shake
root=Delay_0
nodes [
 {id=Delay_0, children [Modulator_0], properties {delayMax=0.1, delayMin=0.1}}
 {id=Modulator_0, children [Random_1]}
 {
  id=Random_1
  children [WavePlayer_20, WavePlayer_21, WavePlayer_22, WavePlayer_23]
  properties {bShouldExcludeFromBranchCulling=true, weights [1,1,1,1]}
 }
 {id=WavePlayer_20, properties {soundWaveAssetPtr=&/Winterfest_Vinyl_Present_Shake_01}}
 {id=WavePlayer_21, properties {soundWaveAssetPtr=&/Winterfest_Vinyl_Present_Shake_02}}
 {id=WavePlayer_22, properties {soundWaveAssetPtr=&/Winterfest_Vinyl_Present_Shake_03}}
 {id=WavePlayer_23, properties {soundWaveAssetPtr=&/Winterfest_Vinyl_Present_Shake_04}}
]
properties {maxDistance=2097152, soundClassObject=&Fort_UI, volumeMultiplier=2, concurrencyOverrides {maxCount=2}}
