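# Winterfest_Ambient_Outside_Cue is a Sound Cue for the looping exterior wind ambience of the Winterfest event.
# A single looping Wave Player feeds the Winterfest_Exterior_Wind_Loop wave straight to the output at reduced volume.
# It is routed through the Fort_UI sound class, has a very large max distance, and keeps playing when silent through its virtualization mode.
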SoundCue /Game/Athena/Sounds/Events/Winterfest/Winterfest_Ambient_Outside_Cue
root=WavePlayer_0
nodes [{id=WavePlayer_0, properties {bLooping=true, soundWaveAssetPtr=&/Winterfest_Exterior_Wind_Loop}}]
properties {
 maxDistance=2097152
 soundClassObject=&Fort_UI
 virtualizationMode=PlayWhenSilent
 volumeMultiplier=0.35
}
