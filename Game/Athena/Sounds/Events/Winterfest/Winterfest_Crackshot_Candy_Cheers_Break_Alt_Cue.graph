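# Sound cue for an alternate Winterfest Crackshot candy cheers break sound effect.
# A single wave player feeds the Winterfest_Cheers_02 wave straight to the output, with no randomization or mixing.
# It uses the Crackshot attenuation asset and the Fort_UI sound class with a boosted volume, and keeps playing when silent.
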SoundCue /Game/Athena/Sounds/Events/Winterfest/Winterfest_Crackshot_Candy_Cheers_Break_Alt_Cue
root=WavePlayer_1
nodes [{id=WavePlayer_1, properties {soundWaveAssetPtr=&/Winterfest_Cheers_02}}]
properties {
 attenuationSettings=&/Winterfest_Crackshot_Att
 maxDistance=1200
 soundClassObject=&Fort_UI
 virtualizationMode=PlayWhenSilent
 volumeMultiplier=1.5
 attenuationOverrides {bUseComplexCollisionForOcclusion=true, lPFRadiusMax=800, lPFRadiusMin=500}
}
