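# Sound cue named for Crackshot's second idle sound in the Winterfest event, with an empty node graph and no root, so it references no audio as saved.
# It is routed to the Fort_UI sound class, uses the shared Winterfest_Crackshot_Att attenuation asset, boosts volume and disables virtualization.
# It also stores attenuation override values for low-pass radii and complex-collision occlusion, though the override is not switched on.
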SoundCue /Game/Athena/Sounds/Events/Winterfest/Winterfest_Crackshot_Idle_02_Cue
root=null
nodes []
properties {
 attenuationSettings=&/Winterfest_Crackshot_Att
 maxDistance=4000
 soundClassObject=&Fort_UI
 virtualizationMode=Disabled
 volumeMultiplier=1.5
 attenuationOverrides {bUseComplexCollisionForOcclusion=true, lPFRadiusMax=800, lPFRadiusMin=500}
}
