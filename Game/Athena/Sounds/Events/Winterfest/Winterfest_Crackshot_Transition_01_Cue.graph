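# Sound cue for a Winterfest Crackshot transition sound.
# A single wave player plays the Winterfest_Leg_Up_01 wave directly as the root, with no randomization or mixing.
# It routes to the Fort_UI sound class with boosted volume and uses the shared Crackshot attenuation asset for distance falloff.
# It keeps playing virtually when silent.
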SoundCue /Game/Athena/Sounds/Events/Winterfest/Winterfest_Crackshot_Transition_01_Cue
root=WavePlayer_0
nodes [{id=WavePlayer_0, properties {soundWaveAssetPtr=&/Winterfest_Leg_Up_01}}]
properties {
 attenuationSettings=&/Winterfest_Crackshot_Att
 maxDistance=1200
 soundClassObject=&Fort_UI
 virtualizationMode=PlayWhenSilent
 volumeMultiplier=1.5
 attenuationOverrides {bUseComplexCollisionForOcclusion=true, lPFRadiusMax=800, lPFRadiusMin=500}
}
