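# Looping Winterfest fireplace ambience cue on the SFX sound class, blending close and far fire layers.
# Two looping waves each pass through their own close or far attenuation asset, so each layer has its own distance response.
# A mixer weights the close layer well above the far one, then a root enveloper ramps the volume up from silence.
# The cue also overrides attenuation at the top level.
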SoundCue /Game/Athena/Sounds/Events/Winterfest/Winterfest_Fireplace_Loop_Cue
root=Enveloper_0
nodes [
 {
  id=Attenuation_0
  children [WavePlayer_1]
  properties {
   attenuationSettings=&/Winterfest_Fire_Loop_Far_Att
   attenuationOverrides {
    distanceAlgorithm=Custom
    falloffDistance=0
    attenuationShapeExtents {x=0}
    customAttenuationCurve {editorCurveData {keys [{interpMode=RCIM_Cubic, tangentMode=RCTM_Break, time=0, value=1}, <at 1 1>]}}
   }
  }
 }
 {
  id=Attenuation_1
  children [WavePlayer_0]
  properties {
   attenuationSettings=&/Winterfest_Fire_Loop_Close_Att
   attenuationOverrides {
    distanceAlgorithm=Custom
    falloffDistance=0
    attenuationShapeExtents {x=0}
    customAttenuationCurve {editorCurveData {keys [{interpMode=RCIM_Cubic, tangentMode=RCTM_Break, time=0, value=1}, {interpMode=RCIM_Cubic, tangentMode=RCTM_Break, time=1, value=0}]}}
   }
  }
 }
 {
  id=Enveloper_0
  children [Mixer_0]
  properties {pitchCurve {editorCurveData {keys [{time=0}]}}, volumeCurve {editorCurveData {keys [<at 0 0>,<at 1 1>]}}}
 }
 {id=Mixer_0, children [Attenuation_1, Attenuation_0], properties {inputVolume [1.5,0.2]}}
 {id=WavePlayer_0, properties {bLooping=true, soundWaveAssetPtr=&/Winterfest_Fireplace_Close_Loop}}
 {id=WavePlayer_1, properties {bLooping=true, soundWaveAssetPtr=&/Winterfest_Fireplace_Far_Loop}}
]
properties {
 bOverrideAttenuation=true
 maxDistance=2097152
 soundClassObject=&SFX
 volumeMultiplier=0.7
 attenuationOverrides {
  stereoSpread=500
  attenuationShapeExtents {x=0}
  customAttenuationCurve {editorCurveData {keys [{interpMode=RCIM_Cubic, tangentMode=RCTM_Break, time=0, value=1}, <at 1 1>]}}
 }
}
