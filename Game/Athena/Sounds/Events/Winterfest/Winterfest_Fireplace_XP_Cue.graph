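# Winterfest fireplace sound cue that plays when XP is gained from a present.
# A mixer layers a common XP-gained sound with a briefly delayed present sound that passes through a modulator.
# An integer XPGain parameter switches that present sound between 10, 20, 100 and 1000 variants, with the 10 variant also serving as default.
# It plays as SFX with overridden short-range attenuation and capped concurrency that stops the oldest.
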SoundCue /Game/Athena/Sounds/Events/Winterfest/Winterfest_Fireplace_XP_Cue
root=Mixer_0
nodes [
 {id=Delay_0, children [Switch_1], properties {delayMax=0.3, delayMin=0.3}}
 {id=Mixer_0, children [Modulator_1,WavePlayer_1], properties {inputVolume [1,1]}}
 {id=Modulator_1, children [Delay_0], properties {pitchMax=1, pitchMin=1, volumeMax=1}}
 {id=Switch_1, children [WavePlayer_8,WavePlayer_8,WavePlayer_5,WavePlayer_7,WavePlayer_4], properties {intParameterName=XPGain}}
 {id=WavePlayer_1, properties {soundWaveAssetPtr=&/UI_Winterfest_XP_Gained}}
 {id=WavePlayer_4, properties {soundWaveAssetPtr=&/Winterfest_XP_Present_1000}}
 {id=WavePlayer_5, properties {soundWaveAssetPtr=&/Winterfest_XP_Present_20}}
 {id=WavePlayer_7, properties {soundWaveAssetPtr=&/Winterfest_XP_Present_100}}
 {id=WavePlayer_8, properties {soundWaveAssetPtr=&/Winterfest_XP_Present_10}}
]
properties {
 bOverrideAttenuation=true
 maxDistance=800
 soundClassObject=&SFX
 volumeMultiplier=0.45
 attenuationOverrides {falloffDistance=300, attenuationShapeExtents {x=500}}
 concurrencyOverrides {maxCount=5, resolutionRule=StopOldest}
}
