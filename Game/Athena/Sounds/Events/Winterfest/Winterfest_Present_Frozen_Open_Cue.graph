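# Sound cue for opening a frozen Winterfest present, routed through the Fort UI sound class.
# After a short fixed delay, a mixer layers a present placement crash with a quieter Winterfest jingle that starts after a second short delay.
# Overall volume is reduced, and a concurrency override caps the cue at two simultaneous instances.
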SoundCue /Game/Athena/Sounds/Events/Winterfest/Winterfest_Present_Frozen_Open_Cue
root=Delay_0
nodes [
 {id=Delay_0, children [Mixer_0], properties {delayMax=0.1, delayMin=0.1}}
 {id=Delay_1, children [WavePlayer_1], properties {delayMax=0.1, delayMin=0.1}}
 {id=Mixer_0, children [WavePlayer_0,Delay_1], properties {inputVolume [1,0.6]}}
 {id=WavePlayer_0, properties {soundWaveAssetPtr=/Game/Sounds/Fort_Gadgets/Present/Present_PlaceCrash_01}}
 {id=WavePlayer_1, properties {soundWaveAssetPtr=&/Winterfest_Jingle_01}}
]
properties {
 bOverrideConcurrency=true
 maxDistance=2097152
 soundClassObject=&Fort_UI
 volumeMultiplier=0.45
 concurrencyOverrides {maxCount=2}
}
