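# Sound cue for the end of a Winterfest present opening, judging by its name.
# A single wave player feeds the output with the Present_PlaceCrash_01 sample from the gadget present sounds, with no randomization or modulation.
# It routes through the Fort_UI sound class at slightly reduced volume and overrides concurrency to cap simultaneous instances at two.
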SoundCue /Game/Athena/Sounds/Events/Winterfest/Winterfest_Present_Opening_End_Cue
root=WavePlayer_0
nodes [{id=WavePlayer_0, properties {soundWaveAssetPtr=/Game/Sounds/Fort_Gadgets/Present/Present_PlaceCrash_01}}]
properties {
 bOverrideConcurrency=true
 maxDistance=2097152
 soundClassObject=&Fort_UI
 volumeMultiplier=0.85
 concurrencyOverrides {maxCount=2}
}
