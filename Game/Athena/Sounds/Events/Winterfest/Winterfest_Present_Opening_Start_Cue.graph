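# Sound cue for the start of opening a Winterfest present, routed through the Fort_UI sound class with boosted volume.
# A fixed short delay leads into an equally weighted random pick between two present-opening waves.
# Both wave players are set to loop.
# A concurrency override allows only one instance at a time, stopping the oldest when a new one starts.
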SoundCue /Game/Athena/Sounds/Events/Winterfest/Winterfest_Present_Opening_Start_Cue
root=Delay_0
nodes [
 {id=Delay_0, children [Random_0], properties {delayMax=0.2, delayMin=0.2}}
 {id=Random_0, children [WavePlayer_0,WavePlayer_1], properties {weights [1,1]}}
 {id=WavePlayer_0, properties {bLooping=true, soundWaveAssetPtr=&/Winterfest_Present_Open_02}}
 {id=WavePlayer_1, properties {bLooping=true, soundWaveAssetPtr=&/Winterfest_Present_Open_01}}
]
properties {
 bOverrideConcurrency=true
 maxDistance=2097152
 soundClassObject=&Fort_UI
 volumeMultiplier=1.5
 concurrencyOverrides {maxCount=1, resolutionRule=StopOldest}
}
