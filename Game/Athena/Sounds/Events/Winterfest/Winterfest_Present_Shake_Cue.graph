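# Sound cue for shaking a Winterfest present, routed through the Fort_UI sound class.
# A short fixed delay feeds a modulator at default settings, then a random node picks one of twenty box and generic shake waves.
# Uneven weights favor a few waves and make many others rare, and selection allows immediate repeats because randomize-without-replacement is off.
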SoundCue /Game/Athena/Sounds/Events/Winterfest/Winterfest_Present_Shake_Cue
root=Delay_0
nodes [
 {id=Delay_0, children [Modulator_0], properties {delayMax=0.1, delayMin=0.1}}
 {id=Modulator_0, children [Random_0]}
 {
  id=Random_0
  children [
   WavePlayer_2
   WavePlayer_0
   WavePlayer_7
   WavePlayer_8
   WavePlayer_9
   WavePlayer_10
   WavePlayer_11
   WavePlayer_12
   WavePlayer_13
   WavePlayer_14
   WavePlayer_15
   WavePlayer_1
   WavePlayer_3
   WavePlayer_4
   WavePlayer_5
   WavePlayer_6
   WavePlayer_16
   WavePlayer_17
   WavePlayer_18
   WavePlayer_19
  ]
  properties {bRandomizeWithoutReplacement=false, weights [1,1,1,0.5,0.1,0.1,0.1,0.5,0.1,0.1,0.5,0.1,0.1,0.5,0.1,0.1,0.5,0.1,0.5,0.5]}
 }
 {id=WavePlayer_0, properties {soundWaveAssetPtr=&/Winterfest_Presents_Shake_Generic_02}}
 {id=WavePlayer_1, properties {soundWaveAssetPtr=&/Winterfest_Presents_Shake_Generic_03}}
 {id=WavePlayer_10, properties {soundWaveAssetPtr=&/Winterfest_Presents_Shake_Generic_12}}
 {id=WavePlayer_11, properties {soundWaveAssetPtr=&/Winterfest_Presents_Shake_Generic_13}}
 {id=WavePlayer_12, properties {soundWaveAssetPtr=&/Winterfest_Presents_Shake_Generic_14}}
 {id=WavePlayer_13, properties {soundWaveAssetPtr=&/Winterfest_Presents_Shake_Generic_15}}
 {id=WavePlayer_14, properties {soundWaveAssetPtr=&/Winterfest_Presents_Shake_Generic_16}}
 {id=WavePlayer_15, properties {soundWaveAssetPtr=&/Winterfest_Presents_Shake_Generic_09}}
 {id=WavePlayer_16, properties {soundWaveAssetPtr=&/Winterfest_Present_Generic_Shake_17}}
 {id=WavePlayer_17, properties {soundWaveAssetPtr=&/Winterfest_Present_Generic_Shake_18}}
 {id=WavePlayer_18, properties {soundWaveAssetPtr=&/Winterfest_Present_Generic_Shake_19}}
 {id=WavePlayer_19, properties {soundWaveAssetPtr=&/Winterfest_Present_Generic_Shake_20}}
 {id=WavePlayer_2, properties {soundWaveAssetPtr=&/Winterfest_Presents_Box_Shake}}
 {id=WavePlayer_3, properties {soundWaveAssetPtr=&/Winterfest_Presents_Shake_Generic_04}}
 {id=WavePlayer_4, properties {soundWaveAssetPtr=&/Winterfest_Presents_Shake_Generic_05}}
 {id=WavePlayer_5, properties {soundWaveAssetPtr=&/Winterfest_Presents_Shake_Generic_06}}
 {id=WavePlayer_6, properties {soundWaveAssetPtr=&/Winterfest_Presents_Shake_Generic_07}}
 {id=WavePlayer_7, properties {soundWaveAssetPtr=&/Winterfest_Presents_Shake_Generic_08}}
 {id=WavePlayer_8, properties {soundWaveAssetPtr=&/Winterfest_Presents_Shake_Generic_10}}
 {id=WavePlayer_9, properties {soundWaveAssetPtr=&/Winterfest_Presents_Shake_Generic_11}}
]
properties {maxDistance=2097152, soundClassObject=&Fort_UI, concurrencyOverrides {maxCount=2}}
