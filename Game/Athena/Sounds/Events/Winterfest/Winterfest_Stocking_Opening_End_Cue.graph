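# Winterfest stocking opening end cue, a layered one-shot sound on the UI sound class with a small concurrency cap.
# A mixer blends four reused waves: a present crash, two holiday tree demolition sounds, and candy cane pickaxe bells.
# The crash and first tree layer start immediately, while the second tree layer follows briefly, slightly pitched down by a modulator.
# The boosted bells enter last, fading in through an enveloper volume curve.
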SoundCue /Game/Athena/Sounds/Events/Winterfest/Winterfest_Stocking_Opening_End_Cue
root=Mixer_0
nodes [
 {id=Delay_1, children [Modulator_0], properties {delayMax=0.15, delayMin=0.15}}
 {id=Delay_2, children [Enveloper_0], properties {delayMax=0.5, delayMin=0.5}}
 {
  id=Enveloper_0
  children [WavePlayer_5]
  properties {pitchCurve {editorCurveData {keys [{time=0}]}}, volumeCurve {editorCurveData {keys [<at 0.0342 0.0079>,<at 0.75 1>]}}}
 }
 {id=Mixer_0, children [WavePlayer_0,WavePlayer_2,Delay_1,Delay_2], properties {inputVolume [1,0.5,1,1.5]}}
 {id=Modulator_0, children [WavePlayer_4], properties {pitchMax=0.95, volumeMax=1, volumeMin=1}}
 {id=WavePlayer_0, properties {soundWaveAssetPtr=/Game/Sounds/Fort_Gadgets/Present/Present_PlaceCrash_01}}
 {id=WavePlayer_2, properties {soundWaveAssetPtr=/Game/Sounds/Fort_Impact_Sounds/Demolition/Tree/Trees_Holiday_Winter_2017-001}}
 {id=WavePlayer_4, properties {soundWaveAssetPtr=/Game/Sounds/Fort_Impact_Sounds/Demolition/Tree/Trees_Holiday_Winter_2017-002}}
 {id=WavePlayer_5, properties {soundWaveAssetPtr=/Game/Athena/Sounds/Weapons/PickAxes/CandyCane/Pickaxe_CandyCane_ImpactBells_03}}
]
properties {
 bOverrideConcurrency=true
 maxDistance=2097152
 soundClassObject=&Fort_UI
 volumeMultiplier=0.8
 concurrencyOverrides {maxCount=2}
}
