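# Sound cue for a Winterfest event transition to the exterior, playing a whoosh away from the cabin.
# A single wave player feeds the root directly, with no randomization or mixing.
# It plays in the UI sound class at boosted volume, with distance attenuation disabled in its overrides.
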SoundCue /Game/Athena/Sounds/Events/Winterfest/Winterfest_Transition_To_Exterior_Cue
root=WavePlayer_0
nodes [{id=WavePlayer_0, properties {soundWaveAssetPtr=&/Winterfest_Whoosh_Away_Cabin}}]
properties {
 maxDistance=2097152
 soundClassObject=&Fort_UI
 volumeMultiplier=1.5
 attenuationOverrides {bAttenuate=false, falloffDistance=800, attenuationShapeExtents {x=500}}
}
