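# Sound cue for the Winterfest event, evidently played on the transition into the cabin interior.
# A single wave player feeds a whoosh-towards-cabin wave directly to the output, with a slight volume boost.
# It routes through the Fort_UI sound class with attenuation disabled, so it plays as a non-spatialized interface sound regardless of listener distance.
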SoundCue /Game/Athena/Sounds/Events/Winterfest/Winterfest_Transition_To_Interior_Cue
root=WavePlayer_1
nodes [{id=WavePlayer_1, properties {soundWaveAssetPtr=&/Winterfest_Whoosh_Towards_Cabin}}]
properties {
 maxDistance=2097152
 soundClassObject=&Fort_UI
 volumeMultiplier=1.2
 attenuationOverrides {bAttenuate=false, falloffDistance=800, attenuationShapeExtents {x=500}}
}
