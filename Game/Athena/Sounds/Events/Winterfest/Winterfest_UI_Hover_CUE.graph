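# Sound cue for a Winterfest UI hover sound, played on the Fort UI sound class.
# A root mixer layers two standard front-end button hover waves, blended in a sub-mixer through a modulator, with a Winterfest hover chime.
# The chime has its own modulator that pitches it sharply up and lowers its volume.
# A concurrency override caps simultaneous instances and blocks new ones, with a short retrigger time to limit rapid repeats.
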SoundCue /Game/Athena/Sounds/Events/Winterfest/Winterfest_UI_Hover_CUE.Winterfest_UI_Hover_Cue
root=Mixer_2
nodes [
 {id=Mixer_0, children [WavePlayer_3,WavePlayer_2], properties {inputVolume [0.75,0.5]}}
 {id=Mixer_2, children [Modulator_0,Modulator_2], properties {inputVolume [1,1]}}
 {id=Modulator_0, children [Mixer_0], properties {volumeMax=1, volumeMin=1}}
 {id=Modulator_2, children [WavePlayer_1], properties {pitchMax=4, pitchMin=4, volumeMax=0.4, volumeMin=0.45}}
 {id=WavePlayer_1, properties {soundWaveAssetPtr=&/UI_Winterfest_HoverChime}}
 {id=WavePlayer_2, properties {soundWaveAssetPtr=/Game/Sounds/Fort_UI_Sounds/FrontEnd/New_FrontEnd/UI_FN_Button_Hover}}
 {id=WavePlayer_3, properties {soundWaveAssetPtr=/Game/Sounds/Fort_UI_Sounds/FrontEnd/New_FrontEnd/UI_FN_Button_Hover_Click_02}}
]
properties {
 bOverrideConcurrency=true
 maxDistance=2097152
 soundClassObject=&Fort_UI
 volumeMultiplier=0.65
 concurrencyOverrides {maxCount=2, resolutionRule=PreventNew, retriggerTime=0.15}
}
