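# Sound cue for a Winterfest UI button press, played through the Fort_UI sound class.
# A Mixer layers the standard front-end button click, boosted in volume, with a much quieter Winterfest click.
# Each layer has its own Modulator with fixed volume, and the Winterfest layer also has its pitch locked.
# A concurrency override caps simultaneous plays and stops the oldest, taking the place of the assigned front-end menu concurrency set.
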SoundCue /Game/Athena/Sounds/Events/Winterfest/Winterfest_UI_Pressed_Cue
root=Mixer_0
nodes [
 {id=Mixer_0, children [Modulator_2,Modulator_1], properties {inputVolume [1,1.2]}}
 {id=Modulator_1, children [WavePlayer_2], properties {pitchMax=1, pitchMin=1, volumeMax=0.2, volumeMin=0.2}}
 {id=Modulator_2, children [WavePlayer_0], properties {volumeMax=1.3, volumeMin=1.3}}
 {id=WavePlayer_0, properties {soundWaveAssetPtr=/Game/Sounds/Fort_UI_Sounds/FrontEnd/New_FrontEnd/UI_FN_Button_Click_Main_01}}
 {id=WavePlayer_2, properties {soundWaveAssetPtr=&/UI_Winterfest_Click_Object}}
]
properties {
 bOverrideConcurrency=true
 maxDistance=2097152
 soundClassObject=&Fort_UI
 volumeMultiplier=1
 concurrencyOverrides {maxCount=4, resolutionRule=StopOldest, retriggerTime=0.2}
 concurrencySet [/Game/Sounds/Concurrency_Rules/FrontEnd_UI_Menu_Max_03_Stop_Oldest]
}
